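# Post-body animation blueprint for the Barefoot backpack cosmetic that makes its tape and pistol bones follow the wearer's upper body.
# It copies the pose from the attached parent mesh, then runs five chained Constraint nodes in component space.
# Each node drives one dyn_ bone from fixed, weighted blends of spine, clavicle or upper-arm twist bones, with no physics simulation.
# The event graph holds only unconnected stub nodes.
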AnimBlueprint /BRCosmetics/Accessories/FORT_Backpacks/Backpack_F_MED_Barefoot/Meshes/F_MED_Barefoot_Skeleton_AnimBlueprint
target_skeleton=&/F_MED_Barefoot_Skeleton
graphs [
 AnimGraph {nodes [
   n1 = "Constraint - dyn_tape_t_l (PPPP)"!(
    ComponentPose=Local To Component(LocalPose=Copy Pose From Mesh() {bUseAttachedParent=true})
    Alpha=1
    ConstraintWeights_0=0.6138
    ConstraintWeights_1:float=0.089
    ConstraintWeights_2:float=0.2941
    ConstraintWeights_3:float=0.003
   ) -> (Pose) {
    alphaBoolBlend {blendOption=Linear}
    boneToModify {boneName=dyn_tape_t_l}
    constraintSetup [
     {targetBone {boneName=clavicle_l}}
     {targetBone {boneName=spine_04}}
     {targetBone {boneName=spine_05}}
     {targetBone {boneName=upperarm_twist_01_l}}
    ]
    constraintWeights [1,1,1,1]
   }
   n2 = "Constraint - dyn_tape_b_l (PPPPPP)"!(
    ComponentPose=@n1
    Alpha=1
    ConstraintWeights_0=0.0754
    ConstraintWeights_1:float=0.0438
    ConstraintWeights_2:float=0.0672
    ConstraintWeights_3:float=0.5317
    ConstraintWeights_4:float=0.2279
    ConstraintWeights_5:float=0.0541
   ) -> (Pose) {
    alphaBoolBlend {blendOption=Linear}
    boneToModify {boneName=dyn_tape_b_l}
    constraintSetup [
     {targetBone {boneName=clavicle_l}}
     {targetBone {boneName=spine_02}}
     {targetBone {boneName=spine_03}}
     {targetBone {boneName=spine_04}}
     {targetBone {boneName=spine_05}}
     {targetBone {boneName=clavicle_r}}
    ]
    constraintWeights [1,1,1,1,1,1]
   }
   n3 = "Constraint - dyn_tape_b_r (PPP)"!(ComponentPose=@n2, Alpha=1, ConstraintWeights_0=0.3068, ConstraintWeights_1:float=0.2642, ConstraintWeights_2:float=0.4291) -> (Pose) {
    alphaBoolBlend {blendOption=Linear}
    boneToModify {boneName=dyn_tape_b_r}
    constraintSetup [{targetBone {boneName=clavicle_r}}, {targetBone {boneName=spine_04}}, {targetBone {boneName=spine_05}}]
    constraintWeights [1,1,1]
   }
   n4 = "Constraint - dyn_tape_t_r (PPP)"!(ComponentPose=@n3, Alpha=1, ConstraintWeights_0=0.0146, ConstraintWeights_1:float=0.7251, ConstraintWeights_2:float=0.2603) -> (Pose) {
    alphaBoolBlend {blendOption=Linear}
    boneToModify {boneName=dyn_tape_t_r}
    constraintSetup [{targetBone {boneName=upperarm_twist_01_r}}, {targetBone {boneName=clavicle_r}}, {targetBone {boneName=spine_05}}]
    constraintWeights [1,1,1]
   }
   n5 = "Constraint - dyn_pistol (PPPP)"!(
    ComponentPose=@n4
    Alpha=1
    ConstraintWeights_0=0.1634
    ConstraintWeights_1:float=0.0388
    ConstraintWeights_2:float=0.6263
    ConstraintWeights_3:float=0.1715
   ) -> (Pose) {
    alphaBoolBlend {blendOption=Linear}
    boneToModify {boneName=dyn_pistol}
    constraintSetup [{targetBone {boneName=clavicle_r}}, {targetBone {boneName=spine_04}}, {targetBone {boneName=spine_05}}, {targetBone {boneName=clavicle_l}}]
    constraintWeights [1,1,1,1]
   }
   Output Pose(Result=Component To Local(ComponentPose=@n5)) {layerGroup=None}
  ]}
 EventGraph {nodes [
   TryGetPawnOwner()
   Event BlueprintUpdateAnimation()
  ]}
]
variables []
